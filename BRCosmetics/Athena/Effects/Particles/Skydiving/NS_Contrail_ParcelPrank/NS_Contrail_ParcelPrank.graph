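# Niagara skydiving contrail for the Parcel Prank cosmetic, combining tumbling mesh cards with two wind ribbons on the player's skeletal mesh.
# Cards spawn near bones in irregular pulses from a looped-age curve with random loop durations, then launch upward through curl noise, their velocity scaling with dive amount.
# Wind_Contrail ribbons follow skeletal positions with upward velocity that switches between gameplay and front end, time-driven material and camera offset.
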NiagaraSystem /BRCosmetics/Athena/Effects/Particles/Skydiving/NS_Contrail_ParcelPrank/NS_Contrail_ParcelPrank
dependencies {usedDataInterfaces [&NiagaraDataInterfaceCurve], usedDynamicInputs [], usedModules [&SystemState]}
emitters {
 Cards {
  emitterSpawnScript [&FN_PreserveSkydiveUserVariables() {module=PreserveUserVariables}]
  emitterUpdateScript [
   &EmitterState(
    Life Cycle Mode=::NewEnumerator1
    Loop Duration=&UniformRangedFloat(Minimum=0.15, Maximum=1.6)
    Recalculate Duration Each Loop=true
    UseLoopDelay=false
   )
   &SpawnRate(
    SpawnRate=&Multiply_Float(
     A=&UniformRangedFloat(Minimum=14, Maximum=32)
     B=&FloatFromCurve(
      FloatCurve=DataInterface() {propertyValues {Curve {keys [
          {arriveTangent=7.6599, interpMode=RCIM_Cubic, leaveTangent=7.6599, tangentMode=RCTM_User, time=0.125, value=0}
          <at 0.4 1>
          {arriveTangent=-2.2897, interpMode=RCIM_Cubic, leaveTangent=-2.2897, tangentMode=RCTM_User, time=0.7, value=1}
          <at 1.4 0>
         ]}}}
      CurveIndex=$Emitter.LoopedAge
     )
    )
    Use Spawn Probability=false
   )
   /Game/Effects/Niagara/Modules/Skydiving/FN_DivingAttributes/FN_EmitterDivingAttributes_v1(Bone Index=2, SkeletalMesh:NiagaraDataInterfaceSkeletalMesh=$User.SkeletalMesh)
  ]
  particleSpawnScript [
   /Niagara/Modules/Spawn/Initialization/InitializeParticle(
    Lifetime=&UniformRangedFloat(Minimum=0.75, Maximum=0.9)
    Mesh Scale=<0.86 0.86 0.86>
    Module.Write Color=true
    Module.Write Lifetime=true
    Module.Write Mass=true
    Module.Write Material Random=true
    Module.Write Position=false
    Module.Write Scale=true
    Module.Write SpriteRotation=false
    Module.Write SpriteSize=false
    Position=&SimulationPosition()
   )
   &SkeletalMeshLocation(
    Bone / Socket Index=&UniformRangedInt(Minimum=0, Maximum=1)
    Bone Sampling Mode=::NewEnumerator4
    Inherit Velocity=false
    MaskSpawnGroup=false
    Skeletal Mesh=$User.SkeletalMesh
   )
   /Niagara/Modules/Spawn/Location/SphereLocation(MaskSpawnGroup=false, Module.Override Randomness=false, Sphere Radius=12, Surface Only=false)
   &AddVelocityInCone(
    Cone Angle=&UniformRangedFloat(Minimum=15, Maximum=20)
    Cone Axis=<0 0 1>
    Module.Use Velocity Falloff On Cone Axis=false
    Velocity Strength=400
   )
   /Niagara/Modules/Spawn/Orientation/V2/InitialMeshOrientation(
    Orientation Vector=<0 1 0>
    Rotation=&UniformRangedVector(Minimum=<0 0 0>, Maximum=<0.5 0 0>)
    Use Orientation Vector=true
    Use Rotation Vector=true
   )
   /SetVariables(Particles.RotationOrientation:NiagaraFloat=&UniformRangedFloat(Minimum=-1, Maximum=1)) {isSetParameters=true}
   /SetVariables(Particles.RotationSpeed:Vector3f=&VectorFromFloat(Value=&UniformAOrBFloat(
      A=&UniformRangedFloat(Minimum=-2, Maximum=-0.65)
      B=&UniformRangedFloat(Minimum=0.65, Maximum=2)
     ))) {isSetParameters=true}
   /SetVariables(Particles.SubImageIndex=&UniformRangedFloat(Minimum=0, Maximum=3)) {isSetParameters=true}
  ]
  particleUpdateScript [
   &ParticleState()
   /Niagara/Modules/Update/Orientation/V2/UpdateMeshOrientation(Rotation Rate:NiagaraFloat=$Particles.RotationOrientation, Rotation Vector:Vector3f=$Particles.RotationSpeed)
   &ScaleVelocity(Velocity Scale=&VectorFromFloat(Value=&Multiply_Float(
      A=&FloatFromCurve(FloatCurve=DataInterface() {propertyValues {Curve {keys [{arriveTangent=1.8262, interpMode=RCIM_Cubic, leaveTangent=1.8262, tangentMode=RCTM_User, time=0.05, value=0.1}, <at 1 1>]}}})
      B=&Lerp_Float(A=1, B=6, Alpha=$Emitter.Dive)
     )))
   &Drag(Use Linear Drag=true, Use Rotational Drag=false)
   /Niagara/Modules/Update/Forces/CurlNoiseForce(Noise Frequency=0.15, Noise Strength=600, Pan Noise=false, Pan Noise Field=<1 1 0>)
   &SolveForcesAndVelocity(Clamp Velocity=false, Limit Acceleration=false)
   &ScaleMeshSize(Scale Factor=&VectorFromFloat(Value=&FloatFromCurve(FloatCurve=DataInterface() {propertyValues {Curve {keys [<at 0 0 RCIM_Cubic>, {arriveTangent=1.1111, leaveTangent=1.1111, time=0.1, value=1}, <at 0.875 1 RCIM_Cubic>, <at 1 0>]}}})))
  ]
  rendererClasses [&NiagaraMeshRendererProperties]
  renderers [{rendererClass=&NiagaraMeshRendererProperties, rendererIndex=0}]
 }
 Wind_Contrail {
  emitterUpdateScript [
   &EmitterState(Life Cycle Mode=::NewEnumerator1, UseLoopDelay=false)
   &FN_SpawnRibbon_v2(Number of Ribbons=2, Spawn Mode=::NewEnumerator0)
  ]
  particleSpawnScript [
   &FN_InitializeRibbon_v3(Color:LinearColor=<rgba 1 1 1 0>, Enable Skeletal Position=true, Ribbon Width=20)
   &AddVelocity(Velocity=&FN_FrontEndVector3Modifier(Gameplay=<0 0 500>, Front End=<0 0 200>))
  ]
  particleUpdateScript [
   /Game/Effects/Niagara/Modules/Skydiving/FN_DivingAttributes/FN_DivingAttributes_v2(Bone Index=2, SkeletalMesh:NiagaraDataInterfaceSkeletalMesh=$User.SkeletalMesh)
   &SolveForcesAndVelocity(Clamp Velocity=false, Limit Acceleration=false)
   &FN_UpdateRibbon(
    Enable Skeletal Position=true
    Initial Color:LinearColor=$Particles.Initial.Color
    Initial RibbonWidth:NiagaraFloat=$Particles.Initial.RibbonWidth
   )
   &Material(Base Texture Time:NiagaraFloat=&Multiply_Float(A=$Emitter.Age, B=-1))
   &ScaleRibbonWidth(Ribbon Width Scale=&FloatFromCurve(FloatCurve=DataInterface() {propertyValues {Curve {keys [<at 0 1 RCIM_Cubic>, {arriveTangent=-0.1012, interpMode=RCIM_Cubic, leaveTangent=-0.1012, time=0.15, value=1.5}, <at 1.0045 0.8984>]}}}))
   &CameraOffset(Camera Offset Amount=10)
  ]
  rendererClasses [&NiagaraRibbonRendererProperties]
  renderers [{rendererClass=&NiagaraRibbonRendererProperties}]
 }
}
summary {
 systemName=NS_Contrail_ParcelPrank
 emitters [
  {emitterName=Cards, rendererClasses [&NiagaraMeshRendererProperties]}
  {emitterName=Wind_Contrail, rendererClasses [&NiagaraRibbonRendererProperties]}
 ]
 userVariables [
  User.bIsFrontEnd {type=NiagaraBool}
  User.bIsFrontEndPreview {type=NiagaraBool}
  User.PlayerSpeed {type=NiagaraFloat, defaultValue=0, description=""}
  User.RibbonIDOffset {type=NiagaraInt32}
  User.SkeletalMesh {
   type=NiagaraDataInterfaceSkeletalMesh
   defaultValue {
    struct=NiagaraExt_VariableValue_DataInterface
    value {dataInterface=&self:NiagaraDataInterfaceSkeletalMesh_0}
   }
  }
 ]
}
